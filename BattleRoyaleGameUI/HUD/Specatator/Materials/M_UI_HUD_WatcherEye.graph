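# Translucent UI material for the spectator HUD that draws a watcher eye procedurally from SphereMasks, without textures.
# Two offset circles intersect into the eye shape widened by WideEye, and inverted masks driven by TopLid and BotLid cut away the lids.
# A parameterized pupil is subtracted from that shape, and a shine offset from the pupil is added back within it.
# A Color parameter tints the result.
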Material /BattleRoyaleGameUI/HUD/Specatator/Materials/M_UI_HUD_WatcherEye
graph {
 properties {blendMode=BLEND_Translucent, materialDomain=MD_UI, pixelDepthOffsetMode=PDOM_Legacy, usedShadingModels=Default Lit}
 nodes [
  Constant() {r=-0.9}
  Constant() {r=-0.9}
  MaterialFunctionCall() {materialFunction=&DebugTimeSine}
  MaterialFunctionCall() {materialFunction=&DebugTimeSine}
  MaterialFunctionCall() {materialFunction=&DebugTimeSine}
  n1 = Subtract(A=Multiply(A=TextureCoordinate()) {constB=2})
  n2 = LinearInterpolate(Alpha=ScalarParameter() {defaultValue=1, parameterName=WideEye}) {constB=0.5}
  n3 = MaterialFunctionCall(X=Constant(), Y=@n2) {materialFunction=&MakeFloat2}
  n4 = Constant() {r=1}
  n5 = Constant() {r=0.98}
  n6 = MaterialFunctionCall(X=Constant(), Y=Multiply(A=@n2) {constB=-1}) {materialFunction=&MakeFloat2}
  n7 = Constant() {r=1}
  n8 = Multiply(A=TextureCoordinate()) {constB=2}
  n9 = LinearInterpolate(Alpha=ScalarParameter() {parameterName=TopLid}) {constA=-3, constB=-0.5}
  n10 = Multiply(A=SphereMask(
    A=Subtract(A=@n8)
    B=MaterialFunctionCall(Y=@n9) {materialFunction=&MakeFloat2}.Result
    Radius=Constant() {r=2}
    Hardness=Constant() {r=0.98}
   ))
  n11 = Multiply(A=TextureCoordinate()) {constB=2}
  n12 = LinearInterpolate(Alpha=ScalarParameter() {parameterName=BotLid}) {constA=2, constB=0.5}
  n13 = Multiply(A=SphereMask(
    A=Subtract(A=@n11)
    B=MaterialFunctionCall(Y=@n12) {materialFunction=&MakeFloat2}.Result
    Radius=Constant() {r=1.2}
    Hardness=Constant() {r=0.98}
   ))
  n14 = Multiply(
   A=Multiply(A=SphereMask(A=@n1, B=@n3.Result, Radius=@n4, Hardness=@n5), B=SphereMask(A=@n1, B=@n6.Result, Radius=@n7, Hardness=@n5))
   B=Multiply(A=OneMinus(None=@n10), B=OneMinus(None=@n13))
  )
  n15 = TextureCoordinate()
  n16 = MaterialFunctionCall(X=ScalarParameter() {parameterName=Pupil_Pos_X}, Y=ScalarParameter() {parameterName=Pupil_Pos_Y}) {materialFunction=&MakeFloat2}
  n17 = SphereMask(
   A=Subtract(A=Multiply(A=@n15) {constB=2})
   B=@n16.Result
   Radius=ScalarParameter() {defaultValue=0.45, parameterName=Pupil_Radius}
   Hardness=ScalarParameter() {defaultValue=0.9, parameterName=Pupil_Hardness}
  )
  n18 = Subtract(A=@n14, B=@n17)
  n19 = Multiply(A=TextureCoordinate()) {constB=2}
  n20 = MaterialFunctionCall(X=ScalarParameter() {defaultValue=0.2, parameterName=Pupil_Shine_X}, Y=ScalarParameter() {defaultValue=-0.3, parameterName=Pupil_Shine_Y}) {materialFunction=&MakeFloat2}
  n21 = Multiply(
   A=@n17
   B=SphereMask(
    A=Subtract(A=@n19)
    B=Add(A=@n16.Result, B=@n20.Result)
    Radius=ScalarParameter() {defaultValue=0.15, parameterName=Pupil_Shine_Radius}
    Hardness=ScalarParameter() {defaultValue=0.93, parameterName=Pupil_Shine_Hardness}
   )
  )
  Multiply(A=VectorParameter() {parameterName=Color, defaultValue=<rgba 1 1 1 1>}.RGB, B=Add(A=Clamp(None=@n18), B=Multiply(A=@n14, B=@n21)))
 ]
}
