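# HUD spectator indicator widget, built on AthenaWatchers, showing a watcher eye beside a count text that is hidden by default.
# Spectator began, added, removed and ended events drive the eye child widget, with separate handling for big groups.
# Count changes play OnValueChanged, an animation with no bindings.
# On Construct, a console variable selects the HUD tags, omitting the post-game-spectate hide tag when enabled.
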WidgetBlueprint /BattleRoyaleGameUI/HUD/Specatator/WBP_BRGameUI_SpecatatorIcon
animations [OnValueChanged {
  binding_count=0
  display_label=""
  length_seconds=0
  object=&self:OnValueChanged
  sequence {bindings []}
 }]
graphs [EventGraph {nodes [
   n1 = Event SpectatorsBegan!() -> (OutputDelegate, then, bBigGroup:bool) {is=title K2Node_Event}
   > Begin_Watcher!(
    execute
    self:WBP BRGame UI HUD Watcher Eye Object Reference=Get WBP_BR_HUD_WatcherEye() -> (WBP_BR_HUD_WatcherEye:WBP BRGame UI HUD Watcher Eye Object Reference)
    ManySpectatorsJoining:bool=@n1.bBigGroup
   ) -> (then)
   n2 = Event SpectatorsAdded!() -> (OutputDelegate, then, bBigGroup:bool) {is=title K2Node_Event}
   > n3 = Branch(Condition=@n2.bBigGroup)
   n4 = Get WBP_BR_HUD_WatcherEye()
   > ManySpectatorsAdded_Watcher(self:WBP BRGame UI HUD Watcher Eye Object Reference=@n4)
   SpectatorAdded_Watcher(execute=@n3.else, self:WBP BRGame UI HUD Watcher Eye Object Reference=@n4)
   Event SpectatorsEnded() {is=title K2Node_Event}
   > End_Watcher(self:WBP BRGame UI HUD Watcher Eye Object Reference=Get WBP_BR_HUD_WatcherEye())
   Event SpectatorRemoved() {is=title K2Node_Event}
   > SpectatorRemoved_Watcher(self:WBP BRGame UI HUD Watcher Eye Object Reference=Get WBP_BR_HUD_WatcherEye())
   Event Spectator Count Changed!() -> (OutputDelegate, then, NewCount:int=0) {is=title K2Node_Event}
   > PlayAnimation(InAnimation=Get OnValueChanged() -> (OnValueChanged:Widget Animation Object Reference))
   Event Construct()
   > n5 = GetConsoleVariableBoolValue!(execute, VariableName:String=Athena.UI.AthenaWatcher.DisplayWidgetForPostGameSpectators) -> (then, ReturnValue:bool)
   > n6 = Branch(Condition=@n5)
   > Set HUDElementTag(HUDElementTag:Gameplay Tag Container Structure=(GameplayTags=((TagName="Athena.HideForCapture"),(TagName="HUD.BottomLeft"),(TagName="HUD.Athena.Hide.Bottom.Left"),(TagName="HUD.Athena.HideForPostGame")))) -> (Output_Get:Gameplay Tag Container Structure)
   Set HUDElementTag(
    execute=@n6.else
    HUDElementTag=(GameplayTags=((TagName="Athena.HideForCapture"),(TagName="HUD.Athena.HideForPostGameSpectate"),(TagName="HUD.BottomLeft"),(TagName="HUD.Athena.Hide.Bottom.Left"),(TagName="HUD.Athena.HideForPostGame")))
   )
  ]}]
variables []
widgets {
 info {inheritedWidgetCount=2, parentClass=/Script/FortniteUI.AthenaWatchers, rootWidgetClass=&SizeBox, widgetCount=5}
 widgets [
  SizeBox_0 {
   widgetClassPath=&SizeBox
   widget=SizeBox_0
   properties {bOverride_WidthOverride=true, heightOverride=23.9984, widthOverride=75.8099}
  }
  HorizontalBox_1 {widgetClassPath=&HorizontalBox, parent=SizeBox_0, widget=HorizontalBox_1, slot {class=SizeBoxSlot}}
  WBP_BR_HUD_WatcherEye {
   widgetClassPath=&/WBP_BRGameUI_HUD_WatcherEye
   bIsVariable=true
   parent=HorizontalBox_1
   widget=WBP_BR_HUD_WatcherEye
   slot {class=HorizontalBoxSlot, properties {padding=<ltrb 0 0 6 0>}}
  }
  Text_Watchers {
   widgetClassPath=&TextBlock
   bInherited=true
   bIsVariable=true
   parent=HorizontalBox_1
   widget=Text_Watchers
   properties {
    text=99
    visibility=Hidden
    font {
     bForceMonospaced=false
     bMaterialIsStencil=false
     fontMaterial=null
     fontObject=&BurbankSmall
     letterSpacing=0
     monospacedWidth=1
     size=13.5
     skewAmount=0
     typefaceFontName=Black
     outlineSettings {
      bApplyOutlineToDropShadows=false
      bMiteredCorners=false
      bSeparateFillAlpha=false
      outlineMaterial=&M-FontShadow
      outlineSize=3
      outlineColor=<rgba 0 0 0 0.5>
     }
    }
   }
   slot {class=HorizontalBoxSlot, properties {verticalAlignment=VAlign_Center}}
  }
  TutorialHighlight {widgetClassPath=&TutorialHighlight}
 ]
}
